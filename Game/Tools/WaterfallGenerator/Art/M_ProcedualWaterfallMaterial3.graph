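# Masked subsurface material for procedural waterfall meshes from the WaterfallGenerator tool, scaled by Blueprint-fed width, height, bevel length and edge offset.
# Remapped UVs separate bevel from fall and drive panning, noise-distorted foam textures, colored by a curve atlas along the flow.
# View-dependent rim glow drives emissive, and Fresnel drives specular.
# Scrolling noise displaces vertices along normals, while temporal dithering fades opacity at both ends of the flow.
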Material /Game/Tools/WaterfallGenerator/Art/M_ProcedualWaterfallMaterial3
graph {
 properties {blendMode=BLEND_Masked, pixelDepthOffsetMode=PDOM_Legacy, shadingModel=MSM_Subsurface, usedShadingModels=Subsurface}
 nodes [
  n1 = NamedRerouteUsage()
  n2 = MaterialFunctionCall(Float2=NamedRerouteUsage()) {materialFunction=&BreakOutFloat2Components}
  n3 = Power(Base=@n2.G, Exp=ScalarParameter() {defaultValue=7, parameterName=CustomParameter_EmissiveRange}) {constExponent=3}
  n4 = OneMinus(None=Power(Base=OneMinus(None=@n2.G), Exp=Constant() {r=0.02}) {constExponent=0.1})
  n5 = Constant() {r=0.5}
  n6 = DotProduct(
   A=MaterialFunctionCall() {materialFunction=&CameraDirectionVector}.Result
   B=Transform(None=Constant3Vector() {constant=<rgba 0 0 1 0>}) {transformSourceType=TRANSFORMSOURCE_Local}
  )
  n7 = Constant() {r=0.5}
  n8 = Power(Base=Saturate(None=Add(A=@n6, B=@n7)), Exp=Constant() {r=1}) {constExponent=3}
  n9 = Multiply(A=@n8, B=Multiply(A=LinearInterpolate(A=@n3, B=@n4, Alpha=@n5), B=OneMinus(None=@n1)))
  n10 = Saturate(None=@n1)
  n11 = Multiply(A=@n10, B=@n10)
  n12 = OneMinus(None=@n10)
  n13 = Multiply(A=@n12, B=@n12)
  Add(A=@n9, B=Multiply(A=Multiply(A=@n11, B=@n13), B=Constant() {r=1}) {constB=4})
  Constant() {r=1}
  Constant() {r=1}
  Constant() {r=200}
  LinearInterpolate(Alpha=NamedRerouteUsage())
  n14 = ScalarParameter() {defaultValue=-0.05, parameterName=CustomParameter_SimpleHueShift}
  n15 = LinearInterpolate(A=NamedRerouteUsage(), Alpha=NamedRerouteUsage())
  MaterialFunctionCall(Hue Shift Percentage=@n14, Texture=@n15) {materialFunction=&HueShift}
  MaterialFunctionCall(Hue Shift Percentage=@n14, Texture=Multiply(A=@n15, B=Constant() {r=0.15}) {constB=0.2}) {materialFunction=&HueShift}
  MaterialFunctionCall(BaseNormal=NamedRerouteUsage(), AdditionalNormal=NamedRerouteUsage()) {materialFunction=&BlendAngleCorrectedNormals}
  n16 = Constant4Vector()
  Multiply(
   A=Desaturation(
    None=NamedRerouteUsage()
    Fraction=MaterialFunctionCall(Value_Morning=@n16.RGBA, Value_Day=@n16.RGBA, Value_Evening=@n16.RGBA, Value_Night=@n16.RGBA) {materialFunction=&Scalar_Time_of_Day}.Result
   )
   B=MaterialFunctionCall(
    Color_Morning=Constant3Vector() {constant=<rgba 0.1 0.0625 0.055 1>}
    Color_Day=Constant3Vector() {constant=<rgba 1 1 1 1>}
    Color_Evening=Constant3Vector() {constant=<rgba 0.1 0.05 0.04 1>}
    Color_Night=Constant3Vector() {constant=<rgba 0.12 0.16 0.2 1>}
   ) {materialFunction=&Color_Time_of_Day}.Result
  )
  n17 = TextureCoordinate()
  n18 = VectorParameter() {
   group="[IgnoreThisGroup]TexUVSetting"
   parameterName=Texture_FoamNoise_A_UVControl
   channelNames {a=VPan, b=UPan, g=VScale, r=UScale}
   defaultValue=<rgba 5 1 0 -0.15>
  }
  n19 = MaterialFunctionCall!(
   Texture=TextureObject() {samplerType=SAMPLERTYPE_LinearGrayscale, texture=&/T_Waterfall_NoiseB}
   "[Override]DDXInput"=DDX(Value=@n17)
   "[Override]DDYInput"=DDY(Value=@n17)
   AdditionalUVOffset
   UVControl_RGB=@n18
   UVControl_A=@n18
   "[Override]TimeInput"
   "[Override]UVInput"=NamedRerouteUsage()
  ) {materialFunction=/Game/Effects/Niagara/MasterMaterials/Functions/Common/FXCommon_TextureSample_DDXDDY}
  n20 = MaterialFunctionCall(R=@n19.R, G=@n19.R) {materialFunction=&AppendMany}
  n21 = VectorParameter() {
   group="[IgnoreThisGroup]NoiseSetting"
   parameterName=Texture_FoamNoise_A_NoiseControl
   sortPriority=60
   channelNames {a=VStrength, b=UStrength, g=VOffset, r=UOffset}
   defaultValue=<rgba 0 -0.5 0 0.05>
  }
  n22 = TextureCoordinate()
  n23 = VectorParameter() {
   group="[IgnoreThisGroup]TexUVSetting"
   parameterName=Texture_FoamNoise_B_UVControl
   channelNames {a=VPan, b=UPan, g=VScale, r=UScale}
   defaultValue=<rgba 5 4 0 -0.1>
  }
  n24 = MaterialFunctionCall(
   Texture=TextureObject() {samplerType=SAMPLERTYPE_Normal, texture=&/T_Waterfall_NormalA}
   "[Override]DDXInput"=DDX(Value=@n22)
   "[Override]DDYInput"=DDY(Value=@n22)
   UVControl_RGB=@n23
   UVControl_A=@n23
   "[Override]UVInput"=NamedRerouteUsage()
  ) {materialFunction=&FXCommon_TextureSample_DDXDDY}
  n25 = VectorParameter() {
   group="[IgnoreThisGroup]NoiseSetting"
   parameterName=Texture_FoamNoise_B_NoiseControl
   sortPriority=60
   channelNames {a=VStrength, b=UStrength, g=VOffset, r=UOffset}
   defaultValue=<rgba 0 0 0 0.01>
  }
  NamedRerouteDeclaration(None=Add(
    A=MaterialFunctionCall!(NoiseValue=@n20.RG, Control_RGB=@n21, Control_A=@n21) {materialFunction=/Game/Effects/Niagara/MasterMaterials/Functions/Common/FXCommon_NoiseControl}.NoiseValue
    B=MaterialFunctionCall(NoiseValue=@n24.RG, Control_RGB=@n25, Control_A=@n25) {materialFunction=&FXCommon_NoiseControl}.NoiseValue
   )) {name=FoamNoise, nodeColor=<rgba 0 1 0.7656 1>}
  n26 = MaterialFunctionCall(Float3=NamedRerouteUsage()) {materialFunction=&BreakOutFloat3Components}
  n27 = MaterialFunctionCall(X=@n26.R, Y=@n26.R, Z=@n26.R) {materialFunction=&MakeFloat3}
  n28 = Constant3Vector() {constant=<rgba 0 -1 0 0>}
  n29 = NamedRerouteUsage()
  n30 = Constant() {r=0.8}
  n31 = Add(A=Multiply(A=DotProduct(A=@n27.Result, B=@n28), B=Multiply(A=@n29, B=@n30)))
  n32 = DotProduct(A=CameraVectorWS(), B=Transform(None=Constant3Vector() {constant=<rgba 0 1 0 0>}) {transformSourceType=TRANSFORMSOURCE_Local})
  n33 = Multiply(
   A=LinearInterpolate(A=OneMinus(None=@n2.R), B=@n2.R, Alpha=ConstantBiasScale(None=@n32)) {constA=-1}
   B=LinearInterpolate(Alpha=Abs(None=@n32)) {constA=2}
  )
  n34 = NamedRerouteUsage()
  n35 = MaterialFunctionCall(Float3=Constant3Vector() {constant=<rgba 5 0.06 0.5 0>}) {materialFunction=&BreakOutFloat3Components}
  n36 = Power(Base=@n34, Exp=@n35.R) {constExponent=3}
  n37 = OneMinus(None=Power(Base=OneMinus(None=@n34), Exp=@n35.G) {constExponent=0.1})
  n38 = Multiply(A=Multiply(A=LinearInterpolate(A=@n36, B=@n37, Alpha=@n35.B), B=@n2.G)) {constB=0.5}
  n39 = NamedRerouteDeclaration(None=Add(A=Multiply(A=@n33, B=@n38), B=@n9)) {name=RimGlow_Mask, nodeColor=<rgba 0 0.9297 1 1>}
  NamedRerouteDeclaration(None=Multiply(A=VectorParameter() {parameterName=CustomParameter_EmissiveColor, defaultValue=<rgba 0 0.1667 0.25 1>}.RGB, B=Multiply(A=@n31, B=@n39))) {name=Emmisive, nodeColor=<rgba 1 0 0.8437 1>}
  n40 = NamedRerouteUsage()
  n41 = NamedRerouteUsage()
  n42 = CurveAtlasRowParameter(CurveTime=MaterialFunctionCall(Float2=Add(A=@n40, B=@n41)) {materialFunction=&BreakOutFloat2Components}.G) {
   atlas=/Game/Athena/Artemis/Environments/BuildingActors/WaterFall/Materials/Texture/CurveAtlas/CLA_WaterFall
   curve=/Game/Athena/Artemis/Environments/BuildingActors/WaterFall/Prefab/S21/CL_PrefabA_ColorCurve1
   defaultValue=21
   group=00.Color
   parameterName=Curve_Color
  }
  NamedRerouteDeclaration(None=@n42.A) {name=Curve_FoamStrength, nodeColor=<rgba 0 1 0.4844 1>}
  NamedRerouteDeclaration(None=@n42) {name=Curve_FoamColor, nodeColor=<rgba 0.2187 1 0 1>}
  n43 = TextureCoordinate()
  n44 = VectorParameter() {
   group="[IgnoreThisGroup]TexUVSetting"
   parameterName=Texture_Foam2_UVControl
   channelNames {a=VPan, b=UPan, g=VScale, r=UScale}
   defaultValue=<rgba 5 2 0 -0.1>
  }
  n45 = Add(A=NamedRerouteUsage(), B=NamedRerouteUsage())
  n46 = MaterialFunctionCall(
   Texture=TextureObject() {texture=/Game/Items/Traps/Textures/T_Barrier_Clouds_03_M}
   "[Override]DDXInput"=DDX(Value=@n43)
   "[Override]DDYInput"=DDY(Value=@n43)
   UVControl_RGB=@n44
   UVControl_A=@n44
   "[Override]UVInput"=@n45
  ) {materialFunction=&FXCommon_TextureSample_DDXDDY}
  n47 = TextureCoordinate()
  n48 = VectorParameter() {
   group="[IgnoreThisGroup]TexUVSetting"
   parameterName=Texture_Foam_UVControl
   channelNames {a=VPan, b=UPan, g=VScale, r=UScale}
   defaultValue=<rgba 6 4.5 0 -0.1>
  }
  n49 = LinearInterpolate(
   A=@n46.R
   B=MaterialFunctionCall(
    Texture=TextureObject() {samplerType=SAMPLERTYPE_LinearColor, texture=/Game/Athena/Apollo/Water/Textures/02_Rounded/03_WaterFoam_Veroni_base_H}
    "[Override]DDXInput"=DDX(Value=@n47)
    "[Override]DDYInput"=DDY(Value=@n47)
    UVControl_RGB=@n48
    UVControl_A=@n48
    "[Override]UVInput"=@n45
   ) {materialFunction=&FXCommon_TextureSample_DDXDDY}.R
   Alpha=ScalarParameter() {defaultValue=0.6, parameterName=CustomParameter_SmoothFoam}
  )
  n50 = NamedRerouteUsage()
  n51 = OneMinus(None=@n50)
  n52 = Constant() {r=0.1}
  n53 = LinearInterpolate(A=Power(Base=@n50, Exp=Constant() {r=1}), B=OneMinus(None=Power(Base=@n51, Exp=@n52)), Alpha=Constant() {r=0.8})
  n54 = NamedRerouteUsage()
  n55 = Subtract(A=Multiply(A=Add(A=@n53, B=@n54)) {constB=2})
  NamedRerouteDeclaration(None=Saturate(None=Add(A=@n49, B=@n55))) {name=FoamMask, nodeColor=<rgba 1 0 0.4219 1>}
  n56 = Constant3Vector() {constant=<rgba 0 0 1 0>}
  NamedRerouteDeclaration(None=Fresnel(ExponentIn=Constant() {r=1}, BaseReflectFractionIn=Constant() {r=0.5}, Normal=Transform(None=@n56))) {name=Specular, nodeColor=<rgba 1 0.3281 0 1>}
  n57 = TextureCoordinate()
  n58 = MaterialFunctionCall(Float2=@n57) {materialFunction=&BreakOutFloat2Components}
  n59 = MaterialFunctionCall(X=OneMinus(None=Frac(None=@n58.R)), Y=@n58.G) {materialFunction=&MakeFloat2}
  NamedRerouteDeclaration(None=If(
    A=MaterialFunctionCall(Float2=@n57) {materialFunction=&BreakOutFloat2Components}.R
    "A > B"=@n59.Result
    "A == B"=@n59.Result
    "A < B"=@n57
   ) {constB=1}) {name=CorrectedUV, nodeColor=<rgba 1 0 0.7734 1>}
  n60 = Constant2Vector() {r=0.5}
  n61 = Add(A=NamedRerouteUsage(), B=Multiply(A=@n60) {constB=-1})
  n62 = NamedRerouteUsage()
  n63 = NamedRerouteUsage()
  n64 = ScalarParameter() {defaultValue=1, parameterName=CustomParameter_TextureDensity}
  n65 = Multiply(A=AppendVector(A=Divide(A=@n62, B=@n63) {constA=500, constB=500}, B=Constant() {r=1}), B=@n64)
  NamedRerouteDeclaration(None=Add(A=Multiply(A=@n61, B=@n65), B=@n60)) {name=RemapedTiledUV, nodeColor=<rgba 0 1 0.0391 1>}
  n66 = Constant() {r=100}
  n67 = NamedRerouteUsage()
  n68 = NamedRerouteUsage()
  n69 = Multiply(
   A=Divide(B=Divide(A=@n66, B=@n67))
   B=OneMinus(None=MaterialFunctionCall(Float2=@n68) {materialFunction=&BreakOutFloat2Components}.G)
  )
  n70 = OneMinus(None=OneMinus(None=Saturate(None=@n69)))
  n71 = Constant() {r=1}
  n72 = MaterialFunctionCall(Alpha Threshold=@n70) {materialFunction=&DitherTemporalAA_Fortnite}
  n73 = NamedRerouteUsage()
  n74 = SmoothStep(
   Max=Divide(A=Multiply(A=@n73) {constB=0.8}, B=NamedRerouteUsage())
   Value=MaterialFunctionCall(Float2=NamedRerouteUsage()) {materialFunction=&BreakOutFloat2Components}.G
  )
  NamedRerouteDeclaration(None=Multiply(
    A=If(A=@n70, B=@n71, "A > B"=@n71, "A == B"=@n71, "A < B"=@n72.Result)
    B=MaterialFunctionCall(Alpha Threshold=@n74) {materialFunction=&DitherTemporalAA}.Result
   )) {name=OpacityMask, nodeColor=<rgba 0.3594 1 0 1>}
  n75 = MaterialFunctionCall(Float2=NamedRerouteUsage()) {materialFunction=&BreakOutFloat2Components}
  n76 = Divide(A=NamedRerouteUsage(), B=NamedRerouteUsage())
  n77 = OneMinus(None=@n76)
  n78 = Divide(A=Subtract(A=@n75.G, B=@n76), B=@n77)
  n79 = LinearInterpolate(
   A=Power(Base=@n78, Exp=ScalarParameter() {defaultValue=0.5, group="[IgnoreThisGroup]UVRemapSetting", parameterName=RemapUV_FallVPower}) {constExponent=0.1}
   B=@n78
   Alpha=Power(Base=OneMinus(None=@n78), Exp=ScalarParameter() {defaultValue=5, group="[IgnoreThisGroup]UVRemapSetting", parameterName=RemapUV_FallVMask})
  )
  n80 = ScalarParameter() {defaultValue=1.2, group="[IgnoreThisGroup]UVRemapSetting", parameterName=RemapUV_BevelVScale}
  n81 = Add(
   A=Multiply(
    A=Multiply(A=@n79, B=@n77)
    B=ScalarParameter() {defaultValue=0.6, group="[IgnoreThisGroup]UVRemapSetting", parameterName=RemapUV_FallVScale}
   )
   B=Multiply(A=@n76, B=@n80)
  )
  n82 = Ceil(None=NamedRerouteUsage())
  n83 = Multiply(A=@n75.G, B=@n80)
  NamedRerouteDeclaration(None=AppendVector(A=@n75.R, B=LinearInterpolate(A=@n81, B=@n83, Alpha=@n82))) {name=RemapedUV, nodeColor=<rgba 0 1 0.9531 1>}
  n84 = Constant2Vector() {r=0.5}
  n85 = Add(A=NamedRerouteUsage(), B=Multiply(A=@n84) {constB=-1})
  n86 = NamedRerouteUsage()
  n87 = NamedRerouteUsage()
  n88 = Multiply(A=AppendVector(A=Divide(A=@n86, B=@n87) {constA=500, constB=500}, B=Constant() {r=1}), B=@n64)
  NamedRerouteDeclaration(None=Add(A=Multiply(A=@n85, B=@n88), B=@n84)) {name=NoRemapedTiledUV, nodeColor=<rgba 0 1 0.0391 1>}
  NamedRerouteDeclaration(None=ScalarParameter() {defaultValue=500, group="[IgnoreThisGroup]BPInput", parameterName=BPInput_Width}) {name=Width, nodeColor=<rgba 1 0 0.6328 1>}
  NamedRerouteDeclaration(None=ScalarParameter() {defaultValue=500, group="[IgnoreThisGroup]BPInput", parameterName=BPInput_Height}) {name=Height, nodeColor=<rgba 1 0 0.1855 1>}
  NamedRerouteDeclaration(None=ScalarParameter() {group="[IgnoreThisGroup]BPInput", parameterName=BPInput_BevelLength}) {name=BevelLength, nodeColor=<rgba 1 0.9609 0 1>}
  n89 = ScalarParameter() {group="[IgnoreThisGroup]BPInput", parameterName=BPInput_EdgeOffset}
  n90 = ScalarParameter() {defaultValue=200, parameterName=CustomParameter_SideAreaSize}
  n91 = Divide(B=Divide(A=Add(A=@n89, B=@n90), B=NamedRerouteUsage()))
  n92 = MaterialFunctionCall(Float2=NamedRerouteUsage()) {materialFunction=&BreakOutFloat2Components}
  n93 = Saturate(None=Multiply(A=@n91, B=@n92.R))
  n94 = Saturate(None=Multiply(A=@n91, B=OneMinus(None=@n92.R)))
  NamedRerouteDeclaration(None=Add(A=OneMinus(None=@n93), B=OneMinus(None=@n94))) {name=AreaMask_Edge, nodeColor=<rgba 0 1 0.7891 1>}
  n95 = NamedRerouteUsage()
  n96 = NamedRerouteUsage()
  n97 = Multiply(
   A=Divide(B=Divide(A=@n95, B=@n96))
   B=MaterialFunctionCall(Float2=NamedRerouteUsage()) {materialFunction=&BreakOutFloat2Components}.G
  )
  NamedRerouteDeclaration(None=OneMinus(None=Saturate(None=@n97))) {name=AreaMask_Bevel, nodeColor=<rgba 0 1 0.7891 1>}
  n98 = TextureCoordinate()
  n99 = VectorParameter() {
   group="[IgnoreThisGroup]TexUVSetting"
   parameterName=Texture_Foam_UVControl
   channelNames {a=VPan, b=UPan, g=VScale, r=UScale}
   defaultValue=<rgba 6 4.5 0 -0.1>
  }
  n100 = MaterialFunctionCall(
   Texture=TextureObject() {samplerType=SAMPLERTYPE_Normal, texture=&/T_Waterfall_NormalB}
   "[Override]DDXInput"=DDX(Value=@n98)
   "[Override]DDYInput"=DDY(Value=@n98)
   UVControl_RGB=@n99
   UVControl_A=@n99
   "[Override]UVInput"=Add(A=NamedRerouteUsage(), B=NamedRerouteUsage())
  ) {materialFunction=&FXCommon_TextureSample_DDXDDY}
  NamedRerouteDeclaration(None=@n100.RGB) {name=FoamNormalOrigin, nodeColor=<rgba 0.2188 0 1 1>}
  NamedRerouteDeclaration(None=MaterialFunctionCall(Normal=@n100.RGB, Flatness=Constant() {r=0.99}) {materialFunction=&FlattenNormal}.Result) {name=FoamNormal, nodeColor=<rgba 1 0.0703 0 1>}
  NamedRerouteDeclaration(None=@n46.R) {name=FoamTexture2Origin, nodeColor=<rgba 0 1 0.2031 1>}
  n101 = MaterialFunctionCall(Float2=NamedRerouteUsage()) {materialFunction=&BreakOutFloat2Components}
  n102 = OneMinus(None=@n101.G)
  n103 = TextureObject() {samplerType=SAMPLERTYPE_LinearGrayscale, texture=&/T_Waterfall_NoiseA}
  n104 = TextureCoordinate()
  n105 = DDX(Value=@n104)
  n106 = DDY(Value=@n104)
  n107 = VectorParameter() {
   group="[IgnoreThisGroup]TexUVSetting"
   parameterName=Texture_MaskNoise_A_UVControl
   channelNames {a=VPan, b=UPan, g=VScale, r=UScale}
   defaultValue=<rgba 1.5 0.4 0 -0.27>
  }
  n108 = NamedRerouteUsage()
  n109 = VectorParameter() {
   group="[IgnoreThisGroup]NoiseSetting"
   parameterName=Texture_MaskNoise_A_NoiseControl
   sortPriority=60
   channelNames {a=VStrength, b=UStrength, g=VOffset, r=UOffset}
   defaultValue=<rgba -0.5 -0.18 0 -0.168>
  }
  n110 = MaterialFunctionCall(
   NoiseValue=Power(
    Base=MaterialFunctionCall(Texture=@n103, "[Override]DDXInput"=@n105, "[Override]DDYInput"=@n106, UVControl_RGB=@n107, UVControl_A=@n107, "[Override]UVInput"=@n108) {materialFunction=&FXCommon_TextureSample_DDXDDY}.R
    Exp=Constant() {r=2}
   )
   Control_RGB=@n109
   Control_A=@n109
  ) {materialFunction=&FXCommon_NoiseControl}
  n111 = Power(Base=@n101.G)
  n112 = TextureObject() {samplerType=SAMPLERTYPE_LinearGrayscale, texture=&T_Waterfall_NoiseB}
  n113 = TextureCoordinate()
  n114 = DDX(Value=@n113)
  n115 = DDY(Value=@n113)
  n116 = VectorParameter() {
   group="[IgnoreThisGroup]TexUVSetting"
   parameterName=Texture_MaskNoise_B_UVControl
   channelNames {a=VPan, b=UPan, g=VScale, r=UScale}
   defaultValue=<rgba 1.75 0.45 0 -0.4>
  }
  n117 = NamedRerouteUsage()
  n118 = VectorParameter() {
   group="[IgnoreThisGroup]NoiseSetting"
   parameterName=Texture_MaskNoise_B_NoiseControl
   sortPriority=60
   channelNames {a=VStrength, b=UStrength, g=VOffset, r=UOffset}
   defaultValue=<rgba -0.5 -0.368 0 -0.7>
  }
  n119 = MaterialFunctionCall(
   NoiseValue=Power(
    Base=MaterialFunctionCall(Texture=@n112, "[Override]DDXInput"=@n114, "[Override]DDYInput"=@n115, UVControl_RGB=@n116, UVControl_A=@n116, "[Override]UVInput"=@n117) {materialFunction=&FXCommon_TextureSample_DDXDDY}.R
    Exp=Constant() {r=1.5}
   )
   Control_RGB=@n118
   Control_A=@n118
  ) {materialFunction=&FXCommon_NoiseControl}
  NamedRerouteDeclaration(None=Add(A=Multiply(A=@n102, B=@n110.NoiseValue), B=Multiply(A=@n111, B=@n119.NoiseValue))) {name=FoamMaskNoise, nodeColor=<rgba 1 0.75 0 1>}
  n120 = TextureObject() {samplerType=SAMPLERTYPE_Normal, texture=/Game/Global/Textures/Default/Tileable_Patterns/T_Tilling_Distortion_Map_01_n}
  n121 = TextureCoordinate()
  n122 = DDX(Value=@n121)
  n123 = DDY(Value=@n121)
  n124 = VectorParameter() {
   group="[IgnoreThisGroup]TexUVSetting"
   parameterName=Texture_WaterNormal_UVControl
   channelNames {a=VPan, b=UPan, g=VScale, r=UScale}
   defaultValue=<rgba 4 0.5 -0.02 -0.2>
  }
  n125 = NamedRerouteUsage()
  NamedRerouteDeclaration(None=MaterialFunctionCall(
    Normal=MaterialFunctionCall(Texture=@n120, "[Override]DDXInput"=@n122, "[Override]DDYInput"=@n123, UVControl_RGB=@n124, UVControl_A=@n124, "[Override]UVInput"=@n125) {materialFunction=&FXCommon_TextureSample_DDXDDY}.RGB
    Flatness=Constant() {r=0.95}
   ) {materialFunction=&FlattenNormal}.Result) {name=WaterlNormal, nodeColor=<rgba 1 0 0.9141 1>}
  n126 = VectorParameter() {
   group="[IgnoreThisGroup]TexUVSetting"
   parameterName=WPOTexture_UVControl
   channelNames {a=VPan, b=UPan, g=VScale, r=UScale}
   defaultValue=<rgba 3 1 0.015 -0.3>
  }
  n127 = Divide(A=NamedRerouteUsage()) {constB=900}
  n128 = Divide(A=NamedRerouteUsage()) {constB=2600}
  MaterialExpressionAppend3Vector_0 = MaterialXAppend3Vector!(A=Round(None=Multiply(A=@n126.UScale, B=@n127)), B=Multiply(A=@n126.VScale, B=@n128), C=Divide(A=@n126.UPan, B=@n127))
  n129 = Subtract(
   A=MaterialFunctionCall(
    Texture=TextureObject() {texture=/Game/Effects/Fort_Effects/Textures/Tile/T_Smooth_Tile_Chromatic}
    UVControl_RGB=@MaterialExpressionAppend3Vector_0
    UVControl_A=Divide(A=@n126.VPan, B=@n128)
    "[Override]UVInput"=NamedRerouteUsage()
   ) {materialFunction=&FXCommon_TextureSample}.R
   B=Constant() {r=0.1}
  ) {constB=0.2}
  n130 = MaterialFunctionCall(Float2=NamedRerouteUsage()) {materialFunction=&BreakOutFloat2Components}
  n131 = Multiply(
   A=MaterialFunctionCall(Input=Power(Base=@n130.G), Input Low=Constant(), Input High=Constant() {r=1}, Target Low=Constant() {r=0.2}, Target High=Constant() {r=1}) {materialFunction=&RemapValueRange}.Result
   B=Multiply(A=ScalarParameter() {defaultValue=1, parameterName=CustomParameter_WPOScale}, B=Constant() {r=100})
  )
  NamedRerouteDeclaration(None=Multiply(A=VertexNormalWS(), B=Multiply(A=@n129, B=@n131))) {name=WPO, nodeColor=<rgba 0.9688 0 1 1>}
  NamedRerouteUsage()
  NamedRerouteUsage()
  NamedRerouteUsage()
  OneMinus(None=@n8)
  OneMinus(None=@n8)
 ]
}
