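# Masked, lit material for the procedural waterfall generator's blend variant, drawing animated foam with dithered, faded edges.
# Blueprint-fed width, height, extend length, bevel and edge offset size the tiled UVs and area masks, with UVs mirrored past U=1.
# Panning noise and normal textures build UV distortion and mask noise for two blended foam samples, using explicit DDX/DDY gradients.
# Temporal-AA dither and depth fade soften coverage, while a V-based mask scales normal-direction vertex offset.
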Material /Game/Tools/WaterfallGenerator/Art/M_ProcedualWaterfallMaterial3_Blend
graph {
 properties {blendMode=BLEND_Masked, pixelDepthOffsetMode=PDOM_Legacy, translucencyLightingMode=TLM_Surface, usedShadingModels=Default Lit}
 nodes [
  Constant() {r=1}
  Constant() {r=1}
  Constant() {r=1}
  Constant() {r=1}
  Constant() {r=1}
  n1 = ScalarParameter() {parameterName=BPInput_ExtendLength}
  Divide(A=NamedRerouteUsage(), B=@n1) {constA=500, constB=500}
  n2 = ScalarParameter() {parameterName=BPInput_EdgeOffset}
  n3 = ScalarParameter() {parameterName=EdgeFade}
  n4 = Divide(B=Divide(A=Add(A=@n2, B=@n3), B=NamedRerouteUsage()))
  n5 = MaterialFunctionCall(Float2=NamedRerouteUsage()) {materialFunction=&BreakOutFloat2Components}
  n6 = Saturate(None=Multiply(A=@n4, B=@n5.R))
  n7 = Saturate(None=Multiply(A=@n4, B=OneMinus(None=@n5.R)))
  n8 = NamedRerouteDeclaration(None=Add(A=OneMinus(None=@n6), B=OneMinus(None=@n7))) {name=AreaMask_Edge_1, nodeColor=<rgba 0 1 0.7891 1>}
  n9 = MaterialFunctionCall(Float2=TextureCoordinate()) {materialFunction=&BreakOutFloat2Components}
  n10 = Constant()
  n11 = Constant() {r=1}
  n12 = Multiply(A=Power(Base=OneMinus(None=@n8)), B=Multiply(A=If(A=@n9.R, "A > B"=@n10, "A == B"=@n10, "A < B"=@n11) {constB=1}, B=NamedRerouteUsage()))
  n13 = MaterialFunctionCall(Alpha Threshold=@n12, Random=Constant() {r=1}) {materialFunction=&DitherTemporalAA_Fortnite}
  If(A=@n12, "A > B"=Constant() {r=1}, "A == B"=@n13.Result, "A < B"=@n13.Result) {constB=0.75}
  n14 = TextureCoordinate()
  n15 = VectorParameter() {
   group="[IgnoreThisGroup]TexUVSetting"
   parameterName=Texture_Foam2_UVControl
   channelNames {a=VPan, b=UPan, g=VScale, r=UScale}
   defaultValue=<rgba 5 2 0 -0.1>
  }
  n16 = Add(A=NamedRerouteUsage(), B=NamedRerouteUsage())
  n17 = TextureCoordinate()
  n18 = VectorParameter() {
   group="[IgnoreThisGroup]TexUVSetting"
   parameterName=Texture_Foam_UVControl
   channelNames {a=VPan, b=UPan, g=VScale, r=UScale}
   defaultValue=<rgba 6 4.5 0 -0.1>
  }
  n19 = MaterialFunctionCall!(
   Texture=TextureObject() {samplerType=SAMPLERTYPE_LinearColor, texture=/Game/Athena/Apollo/Water/Textures/02_Rounded/03_WaterFoam_Veroni_base_H}
   "[Override]DDXInput"=DDX(Value=@n17)
   "[Override]DDYInput"=DDY(Value=@n17)
   AdditionalUVOffset
   UVControl_RGB=@n18
   UVControl_A=@n18
   "[Override]TimeInput"
   "[Override]UVInput"=@n16
  ) {materialFunction=/Game/Effects/Niagara/MasterMaterials/Functions/Common/FXCommon_TextureSample_DDXDDY}
  LinearInterpolate(
   A=MaterialFunctionCall(
    Texture=TextureObject() {texture=/Game/Items/Traps/Textures/T_Barrier_Clouds_03_M}
    "[Override]DDXInput"=DDX(Value=@n14)
    "[Override]DDYInput"=DDY(Value=@n14)
    UVControl_RGB=@n15
    UVControl_A=@n15
    "[Override]UVInput"=@n16
   ) {materialFunction=&FXCommon_TextureSample_DDXDDY}.R
   B=@n19.R
   Alpha=ScalarParameter() {defaultValue=0.6, parameterName=CustomParameter_SmoothFoam}
  )
  MaterialFunctionCall(Emissive=Constant() {r=0.05}) {materialFunction=&MF_FX_EmissiveAdjustments}
  Multiply(A=DepthFade() {fadeDistanceDefault=5}, B=@n12)
  n20 = MaterialFunctionCall(Float2=NamedRerouteUsage()) {materialFunction=&BreakOutFloat2Components}
  n21 = ScalarParameter() {defaultValue=0.7}
  n22 = Multiply(A=Saturate(None=Subtract(A=@n20.G, B=@n21)), B=Divide(B=@n21))
  Multiply(A=Multiply(A=Saturate(None=@n22), B=VertexNormalWS()), B=ScalarParameter() {defaultValue=-10, parameterName=WPO})
  n23 = TextureCoordinate()
  n24 = VectorParameter() {
   group="[IgnoreThisGroup]TexUVSetting"
   parameterName=Texture_FoamNoise_A_UVControl
   channelNames {a=VPan, b=UPan, g=VScale, r=UScale}
   defaultValue=<rgba 5 1 0 -0.15>
  }
  n25 = MaterialFunctionCall(
   Texture=TextureObject() {samplerType=SAMPLERTYPE_LinearGrayscale, texture=&/T_Waterfall_NoiseB}
   "[Override]DDXInput"=DDX(Value=@n23)
   "[Override]DDYInput"=DDY(Value=@n23)
   UVControl_RGB=@n24
   UVControl_A=@n24
   "[Override]UVInput"=NamedRerouteUsage()
  ) {materialFunction=&FXCommon_TextureSample_DDXDDY}
  n26 = MaterialFunctionCall(R=@n25.R, G=@n25.R) {materialFunction=&AppendMany}
  n27 = VectorParameter() {
   group="[IgnoreThisGroup]NoiseSetting"
   parameterName=Texture_FoamNoise_A_NoiseControl
   sortPriority=60
   channelNames {a=VStrength, b=UStrength, g=VOffset, r=UOffset}
   defaultValue=<rgba 0 -0.5 0 0.05>
  }
  n28 = TextureCoordinate()
  n29 = VectorParameter() {
   group="[IgnoreThisGroup]TexUVSetting"
   parameterName=Texture_FoamNoise_B_UVControl
   channelNames {a=VPan, b=UPan, g=VScale, r=UScale}
   defaultValue=<rgba 5 4 0 -0.1>
  }
  n30 = MaterialFunctionCall(
   Texture=TextureObject() {samplerType=SAMPLERTYPE_Normal, texture=&/T_Waterfall_NormalA}
   "[Override]DDXInput"=DDX(Value=@n28)
   "[Override]DDYInput"=DDY(Value=@n28)
   UVControl_RGB=@n29
   UVControl_A=@n29
   "[Override]UVInput"=NamedRerouteUsage()
  ) {materialFunction=&FXCommon_TextureSample_DDXDDY}
  n31 = VectorParameter() {
   group="[IgnoreThisGroup]NoiseSetting"
   parameterName=Texture_FoamNoise_B_NoiseControl
   sortPriority=60
   channelNames {a=VStrength, b=UStrength, g=VOffset, r=UOffset}
   defaultValue=<rgba 0 0 0 0.01>
  }
  NamedRerouteDeclaration(None=Add(
    A=MaterialFunctionCall!(NoiseValue=@n26.RG, Control_RGB=@n27, Control_A=@n27) {materialFunction=/Game/Effects/Niagara/MasterMaterials/Functions/Common/FXCommon_NoiseControl}.NoiseValue
    B=MaterialFunctionCall(NoiseValue=@n30.RG, Control_RGB=@n31, Control_A=@n31) {materialFunction=&FXCommon_NoiseControl}.NoiseValue
   )) {name=FoamNoise, nodeColor=<rgba 0 1 0.7656 1>}
  n32 = Divide(A=Multiply(A=Subtract(A=@n20.G) {constB=0.5}) {constB=2})
  n33 = Power(Base=Add(A=@n22, B=Add(A=@n19.R, B=@n32)))
  NamedRerouteDeclaration(None=MaterialFunctionCall(In=Saturate(None=@n33), Contrast=Constant() {r=1}) {materialFunction=&CheapContrast}.Result) {name=FoamMask, nodeColor=<rgba 1 0 0.4219 1>}
  n34 = TextureCoordinate()
  n35 = MaterialFunctionCall(Float2=@n34) {materialFunction=&BreakOutFloat2Components}
  n36 = MaterialFunctionCall(X=OneMinus(None=Frac(None=@n35.R)), Y=@n35.G) {materialFunction=&MakeFloat2}
  NamedRerouteDeclaration(None=If(
    A=MaterialFunctionCall(Float2=@n34) {materialFunction=&BreakOutFloat2Components}.R
    "A > B"=@n36.Result
    "A == B"=@n36.Result
    "A < B"=@n34
   ) {constB=1}) {name=CorrectedUV, nodeColor=<rgba 1 0 0.7734 1>}
  n37 = Constant2Vector() {r=0.5}
  n38 = Add(A=NamedRerouteUsage(), B=Multiply(A=@n37) {constB=-1})
  n39 = Divide(A=NamedRerouteUsage()) {constA=500, constB=500}
  n40 = Divide(A=@n1) {constA=500, constB=500}
  n41 = VectorParameter() {parameterName=BlendAreaScaleUV, defaultValue=<rgba 0 0 0 1>}
  n42 = Multiply(A=ScalarParameter() {defaultValue=1, parameterName=CustomParameter_TextureDensity}, B=AppendVector(A=@n41.R, B=@n41.R))
  n43 = Multiply(A=AppendVector(A=@n39, B=@n40), B=@n42)
  NamedRerouteDeclaration(None=Add(A=Multiply(A=@n38, B=@n43), B=@n37)) {name=RemapedTiledUV, nodeColor=<rgba 0 1 0.0391 1>}
  NamedRerouteDeclaration(None=NamedRerouteUsage()) {name=RemapedUV, nodeColor=<rgba 0 1 0.9531 1>}
  n44 = Constant2Vector() {r=0.5}
  n45 = Add(A=NamedRerouteUsage(), B=Multiply(A=@n44) {constB=-1})
  n46 = Multiply(A=AppendVector(A=@n39, B=@n40), B=@n42)
  NamedRerouteDeclaration(None=Add(A=Multiply(A=@n45, B=@n46), B=@n44)) {name=NoRemapedTiledUV, nodeColor=<rgba 0 1 0.0391 1>}
  NamedRerouteDeclaration(None=ScalarParameter() {defaultValue=500, group="[IgnoreThisGroup]BPInput", parameterName=BPInput_Width}) {name=Width, nodeColor=<rgba 1 0 0.6328 1>}
  NamedRerouteDeclaration(None=ScalarParameter() {defaultValue=500, group="[IgnoreThisGroup]BPInput", parameterName=BPInput_Height}) {name=Height, nodeColor=<rgba 1 0 0.1855 1>}
  NamedRerouteDeclaration(None=ScalarParameter() {group="[IgnoreThisGroup]BPInput", parameterName=BPInput_BevelLength}) {name=BevelLength, nodeColor=<rgba 1 0.9609 0 1>}
  n47 = ScalarParameter() {group="[IgnoreThisGroup]BPInput", parameterName=BPInput_EdgeOffset}
  n48 = ScalarParameter() {defaultValue=200, parameterName=CustomParameter_SideAreaSize}
  n49 = Divide(B=Divide(A=Add(A=@n47, B=@n48), B=NamedRerouteUsage()))
  n50 = MaterialFunctionCall(Float2=NamedRerouteUsage()) {materialFunction=&BreakOutFloat2Components}
  n51 = Saturate(None=Multiply(A=@n49, B=@n50.R))
  n52 = Saturate(None=Multiply(A=@n49, B=OneMinus(None=@n50.R)))
  NamedRerouteDeclaration(None=Add(A=OneMinus(None=@n51), B=OneMinus(None=@n52))) {name=AreaMask_Edge, nodeColor=<rgba 0 1 0.7891 1>}
  n53 = NamedRerouteUsage()
  n54 = NamedRerouteUsage()
  n55 = Multiply(
   A=Divide(B=Divide(A=@n53, B=@n54))
   B=MaterialFunctionCall(Float2=NamedRerouteUsage()) {materialFunction=&BreakOutFloat2Components}.G
  )
  NamedRerouteDeclaration(None=OneMinus(None=Saturate(None=@n55))) {name=AreaMask_Bevel, nodeColor=<rgba 0 1 0.7891 1>}
  n56 = MaterialFunctionCall(Float2=NamedRerouteUsage()) {materialFunction=&BreakOutFloat2Components}
  n57 = OneMinus(None=@n56.G)
  n58 = TextureObject() {samplerType=SAMPLERTYPE_LinearGrayscale, texture=&/T_Waterfall_NoiseA}
  n59 = TextureCoordinate()
  n60 = DDX(Value=@n59)
  n61 = DDY(Value=@n59)
  n62 = VectorParameter() {
   group="[IgnoreThisGroup]TexUVSetting"
   parameterName=Texture_MaskNoise_A_UVControl
   channelNames {a=VPan, b=UPan, g=VScale, r=UScale}
   defaultValue=<rgba 1.5 0.4 0 -0.27>
  }
  n63 = NamedRerouteUsage()
  n64 = VectorParameter() {
   group="[IgnoreThisGroup]NoiseSetting"
   parameterName=Texture_MaskNoise_A_NoiseControl
   sortPriority=60
   channelNames {a=VStrength, b=UStrength, g=VOffset, r=UOffset}
   defaultValue=<rgba -0.5 -0.18 0 -0.168>
  }
  n65 = MaterialFunctionCall(
   NoiseValue=Power(
    Base=MaterialFunctionCall(Texture=@n58, "[Override]DDXInput"=@n60, "[Override]DDYInput"=@n61, UVControl_RGB=@n62, UVControl_A=@n62, "[Override]UVInput"=@n63) {materialFunction=&FXCommon_TextureSample_DDXDDY}.R
    Exp=Constant() {r=2}
   )
   Control_RGB=@n64
   Control_A=@n64
  ) {materialFunction=&FXCommon_NoiseControl}
  n66 = Power(Base=@n56.G)
  n67 = TextureObject() {samplerType=SAMPLERTYPE_LinearGrayscale, texture=&T_Waterfall_NoiseB}
  n68 = TextureCoordinate()
  n69 = DDX(Value=@n68)
  n70 = DDY(Value=@n68)
  n71 = VectorParameter() {
   group="[IgnoreThisGroup]TexUVSetting"
   parameterName=Texture_MaskNoise_B_UVControl
   channelNames {a=VPan, b=UPan, g=VScale, r=UScale}
   defaultValue=<rgba 1.75 0.45 0 -0.4>
  }
  n72 = NamedRerouteUsage()
  n73 = VectorParameter() {
   group="[IgnoreThisGroup]NoiseSetting"
   parameterName=Texture_MaskNoise_B_NoiseControl
   sortPriority=60
   channelNames {a=VStrength, b=UStrength, g=VOffset, r=UOffset}
   defaultValue=<rgba -0.5 -0.368 0 -0.7>
  }
  n74 = MaterialFunctionCall(
   NoiseValue=Power(
    Base=MaterialFunctionCall(Texture=@n67, "[Override]DDXInput"=@n69, "[Override]DDYInput"=@n70, UVControl_RGB=@n71, UVControl_A=@n71, "[Override]UVInput"=@n72) {materialFunction=&FXCommon_TextureSample_DDXDDY}.R
    Exp=Constant() {r=1.5}
   )
   Control_RGB=@n73
   Control_A=@n73
  ) {materialFunction=&FXCommon_NoiseControl}
  NamedRerouteDeclaration(None=Add(A=Multiply(A=@n57, B=@n65.NoiseValue), B=Multiply(A=@n66, B=@n74.NoiseValue))) {name=FoamMaskNoise, nodeColor=<rgba 1 0.75 0 1>}
  NamedRerouteUsage()
  NamedRerouteUsage()
 ]
}
