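# Translucent, lit Niagara sprite material for the waterfall generator's drop effect, drawing a noise-distorted smoke shape tinted by particle color.
# UVs are split into 2x2 cells, scaled about each cell center with a vertical taper, then offset by two panning noise layers, the second scaled by a dynamic parameter.
# A tiling-method mask confines the shape to its cell.
# Particle alpha, add/multiply/power adjustments and depth fade shape opacity.
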Material /Game/Tools/WaterfallGenerator/Art/M_Waterfall_Drop
graph {
 properties {
  bEnableResponsiveAA=true
  bUsedWithNiagaraSprites=true
  blendMode=BLEND_Translucent
  pixelDepthOffsetMode=PDOM_Legacy
  translucencyLightingMode=TLM_VolumetricPerVertexNonDirectional
  usedShadingModels=Default Lit
 }
 nodes [
  Constant() {r=1}
  n1 = ParticleColor()
  n2 = TextureObjectParameter() {
   group=Shape
   parameterName=Shape_Texture
   samplerType=SAMPLERTYPE_LinearGrayscale
   sortPriority=0
   texture=&/T_Waterfall_SmokeShape
  }
  n3 = VectorParameter() {group=Shape, parameterName=Shape_UVControl, channelNames {a=VPan, b=UPan, g=VScale, r=UScale}, defaultValue=<rgba 1 1 0 0>}
  n4 = TextureObjectParameter() {
   group=NoiseA
   parameterName=NoiseA_Texture
   samplerType=SAMPLERTYPE_LinearGrayscale
   sortPriority=0
   texture=&/T_Waterfall_NoiseB
  }
  n5 = VectorParameter() {group=NoiseA, parameterName=NoiseA_UVControl, channelNames {a=VPan, b=UPan, g=VScale, r=UScale}, defaultValue=<rgba 2 0.6 0 0.25>}
  n6 = MaterialFunctionCall(Texture=@n4, UVControl_RGB=@n5, UVControl_A=@n5) {materialFunction=&FXCommon_TextureSample}
  n7 = VectorParameter() {
   group=NoiseA
   parameterName=NoiseA_NoiseControl
   sortPriority=60
   channelNames {a=VStrength, b=UStrength, g=VOffset, r=UOffset}
   defaultValue=<rgba -0.5 -0.5 0.01 0.2>
  }
  n8 = TextureObjectParameter() {
   group=NoiseB
   parameterName=NoiseB_Texture
   samplerType=SAMPLERTYPE_LinearGrayscale
   sortPriority=0
   texture=&T_SimplePerlin_Greyscale
  }
  n9 = VectorParameter() {group=NoiseB, parameterName=NoiseB_UVControl, channelNames {a=VPan, b=UPan, g=VScale, r=UScale}, defaultValue=<rgba 3 0.5 0 0.5>}
  n10 = VectorParameter() {
   group=NoiseB
   parameterName=NoiseB_NoiseControl
   sortPriority=60
   channelNames {a=VStrength, b=UStrength, g=VOffset, r=UOffset}
   defaultValue=<rgba -0.5 -0.5 0 0.3>
  }
  n11 = MaterialFunctionCall!(
   NoiseValue=MaterialFunctionCall(Texture=@n8, UVControl_RGB=@n9, UVControl_A=@n9) {materialFunction=&FXCommon_TextureSample}.RG
   Control_RGB=@n10
   Control_A=@n10
  ) {materialFunction=/Game/Effects/Niagara/MasterMaterials/Functions/Common/FXCommon_NoiseControl}
  n12 = DynamicParameter()
  n13 = Multiply(A=TextureCoordinate()) {constB=1.9999}
  n14 = Multiply(A=Floor(None=@n13)) {constB=0.5}
  n15 = Add(A=@n14) {constB=0.25}
  n16 = Subtract(A=TextureCoordinate(), B=@n15) {constB=0.5}
  n17 = VectorParameter() {parameterName=A, defaultValue=<rgba 1 0.2 2.5 0.5>}
  n18 = MaterialFunctionCall(Float2=Frac(None=@n13)) {materialFunction=&BreakOutFloat2Components}
  n19 = Multiply(
   A=MaterialFunctionCall(R=LinearInterpolate(A=@n17.R, B=@n17.R, Alpha=@n18.G) {constA=2, constB=0}, G=@n17.A) {materialFunction=&AppendMany}.RG
   B=@n17.B
  )
  n20 = Add(
   A=Add(
    A=MaterialFunctionCall(NoiseValue=@n6.RG, Control_RGB=@n7, Control_A=@n7) {materialFunction=&FXCommon_NoiseControl}.NoiseValue
    B=Multiply(A=@n11.NoiseValue, B=@n12.Param1)
   )
   B=Add(A=Multiply(A=@n16, B=@n19) {constB=0.9}, B=@n15) {constB=0.5}
  )
  n21 = Subtract(A=@n20, B=@n14)
  n22 = Multiply(
   A=Multiply(
    A=@n1.A
    B=MaterialFunctionCall(Texture=@n2, UVControl_RGB=@n3, UVControl_A=@n3, "[Override]UVInput"=@n20) {materialFunction=&FXCommon_TextureSample}.R
   )
   B=MaterialFunctionCall!(TexCoord=Multiply(A=@n21) {constB=2}, ClampChannel=Constant() {r=1}) {materialFunction=/Game/Effects/Fort_Effects/MaterialFunctions/MF_FX_TilingMethod}.Result
  )
  n23 = ScalarParameter() {group=Opacity, parameterName=Opacity_GrayAdjust_Add}
  n24 = Saturate(None=Multiply(A=Add(A=@n22, B=@n23), B=ScalarParameter() {defaultValue=2, group=Opacity, parameterName=Opacity_GrayAdjust_Mul}))
  n25 = ScalarParameter() {defaultValue=1.5, group=Opacity, parameterName=Opacity_GrayAdjust_Power}
  DepthFade(Opacity=Saturate(None=Power(Base=@n24, Exp=@n25)), FadeDistance=ScalarParameter() {defaultValue=100, parameterName=DF})
  MaterialFunctionCall(Emissive=@n1.RGB) {materialFunction=&MF_FX_EmissiveAdjustments}
  Normalize(VectorInput=@n1.RGB)
 ]
}
