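# Translucent lit Niagara sprite material for waterfall edges in the WaterfallGenerator tool.
# Two panning noise textures distort the UVs of a ribbon-shape texture, which are also scaled about their center by a power curve along U.
# The shape, multiplied by a U-based edge mask, passes through depth fade and particle alpha for soft opacity.
# Particle color supplies the tint and emissive adjustment, while the shape also blends a roughness-like scalar.
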Material /Game/Tools/WaterfallGenerator/Art/M_Waterfall_Edge
graph {
 properties {
  bEnableResponsiveAA=true
  bUsedWithNiagaraSprites=true
  blendMode=BLEND_Translucent
  pixelDepthOffsetMode=PDOM_Legacy
  translucencyLightingMode=TLM_VolumetricPerVertexNonDirectional
  usedShadingModels=Default Lit
 }
 nodes [
  Constant() {r=1}
  n1 = TextureObject() {samplerType=SAMPLERTYPE_Grayscale, texture=/Game/Gadgets/Assets/VinderTech_GliderChute/Glider_Thrive_Female/FX/Texture/T_RibbonBaseShape4}
  n2 = VectorParameter() {group=MainTexture, parameterName=MainTexture_UVControl, channelNames {a=VPan, b=UPan, g=VScale, r=UScale}, defaultValue=<rgba 1 1 0 0>}
  n3 = TextureObjectParameter() {
   group=NoiseA
   parameterName=NoiseA_Texture
   samplerType=SAMPLERTYPE_LinearGrayscale
   sortPriority=0
   texture=&T_SimplePerlin_Greyscale
  }
  n4 = VectorParameter() {group=NoiseA, parameterName=NoiseA_UVControl, channelNames {a=VPan, b=UPan, g=VScale, r=UScale}, defaultValue=<rgba 0.7 1.5 0.7 0>}
  n5 = MaterialFunctionCall(Texture=@n3, UVControl_RGB=@n4, UVControl_A=@n4) {materialFunction=&FXCommon_TextureSample}
  n6 = VectorParameter() {
   group=NoiseA
   parameterName=NoiseA_NoiseControl
   sortPriority=60
   channelNames {a=VStrength, b=UStrength, g=VOffset, r=UOffset}
   defaultValue=<rgba -0.34 -0.5 2 0.5>
  }
  n7 = MaterialFunctionCall!(NoiseValue=AppendVector(A=@n5.R, B=@n5.R), Control_RGB=@n6, Control_A=@n6) {materialFunction=/Game/Effects/Niagara/MasterMaterials/Functions/Common/FXCommon_NoiseControl}
  n8 = MaterialFunctionCall(Float2=TextureCoordinate()) {materialFunction=&BreakOutFloat2Components}
  n9 = Add(A=@n8.R) {constB=0}
  n10 = MaterialFunctionCall(
   UVs=TextureCoordinate()
   Texture Scale=LinearInterpolate(B=Constant2Vector() {r=1}, Alpha=Power(Base=@n9) {constExponent=4}) {constA=1, constB=0}
  ) {materialFunction=&ScaleUVsByCenter}
  n11 = TextureObjectParameter() {
   group=NoiseB
   parameterName=NoiseB_Texture
   samplerType=SAMPLERTYPE_LinearGrayscale
   sortPriority=0
   texture=&/T_Waterfall_NoiseB
  }
  n12 = VectorParameter() {group=NoiseB, parameterName=NoiseB_UVControl, channelNames {a=VPan, b=UPan, g=VScale, r=UScale}, defaultValue=<rgba 0.75 1.5 0.8 0>}
  n13 = MaterialFunctionCall(Texture=@n11, UVControl_RGB=@n12, UVControl_A=@n12) {materialFunction=&FXCommon_TextureSample}
  n14 = AppendVector(A=@n13.R, B=@n13.R)
  n15 = VectorParameter() {
   group=NoiseB
   parameterName=NoiseB_NoiseControl
   sortPriority=60
   channelNames {a=VStrength, b=UStrength, g=VOffset, r=UOffset}
   defaultValue=<rgba -0.3 -0.5 2 0.25>
  }
  n16 = MaterialFunctionCall(
   Texture=@n1
   UVControl_RGB=@n2
   UVControl_A=@n2
   "[Override]UVInput"=Add(
    A=Add(A=@n7.NoiseValue, B=@n10.UVs)
    B=MaterialFunctionCall(NoiseValue=@n14, Control_RGB=@n15, Control_A=@n15) {materialFunction=&FXCommon_NoiseControl}.NoiseValue
   )
  ) {materialFunction=&FXCommon_TextureSample}
  LinearInterpolate(B=ScalarParameter() {defaultValue=0.9, parameterName=R}, Alpha=@n16.G) {constA=1, constB=0.9}
  n17 = ParticleColor()
  MaterialFunctionCall(Emissive=@n17.RGB) {materialFunction=&MF_FX_EmissiveAdjustments}
  n18 = Multiply(A=@n16.G, B=ScalarParameter() {defaultValue=1})
  n19 = VectorParameter() {group=Mask, parameterName=Mask_EdgeMask_InvertControl, channelNames {a="", b=Power, g=Multiply, r=Offset}, defaultValue=<rgba 0 5 2 1>}
  n20 = VectorParameter() {group=Mask, parameterName=Mask_EdgeMask_Control, channelNames {a="", b=Power, g=Multiply, r=Offset}, defaultValue=<rgba 0 2 1.5 1>}
  n21 = MaterialFunctionCall!(UorV=@n8.R, Invert_EdgeControl=@n19, NoInvert_EdgeControl=@n20) {materialFunction=/Game/Effects/Niagara/MasterMaterials/Functions/Common/FXCommon_EdgeMask_Ex}
  Multiply(A=@n17.A, B=DepthFade(Opacity=Multiply(A=@n18, B=@n21.Mask)) {fadeDistanceDefault=30})
  Normalize(VectorInput=@n17.RGB)
 ]
}
