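# Translucent, lit Niagara sprite material for flying sand particles in the waterfall generator, with responsive AA.
# Remapped red and panning green channels of a sand normal map form the grain, shaped by a sphere mask with Perlin-distorted UVs.
# A per-particle random time offset varies the panning.
# Particle alpha, depth fade and two-axis edge masks soften opacity, while particle color feeds an emissive adjustment function.
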Material /Game/Tools/WaterfallGenerator/Art/M_Waterfall_FlyingSand
graph {
 properties {
  bEnableResponsiveAA=true
  bUsedWithNiagaraSprites=true
  blendMode=BLEND_Translucent
  pixelDepthOffsetMode=PDOM_Legacy
  translucencyLightingMode=TLM_VolumetricPerVertexNonDirectional
  usedShadingModels=Default Lit
 }
 nodes [
  n1 = ParticleColor()
  MaterialFunctionCall(Emissive=@n1.RGB) {materialFunction=&MF_FX_EmissiveAdjustments}
  n2 = MaterialFunctionCall(Float2=TextureCoordinate()) {materialFunction=&BreakOutFloat2Components}
  n3 = VectorParameter() {group=Mask, parameterName=Mask_EdgeMask_Control, channelNames {a="", b=Power, g=Multiply, r=Offset}, defaultValue=<rgba 0 10 1.5 1>}
  n4 = TextureCoordinate()
  n5 = Add(A=TextureSample(UVs=Multiply(A=@n4) {constB=0.5}) {samplerType=SAMPLERTYPE_Normal, texture=&/T_Waterfall_Sand}.R)
  n6 = TextureObjectParameter() {
   group=TextureB
   parameterName=TextureB_Texture
   samplerType=SAMPLERTYPE_Normal
   sortPriority=0
   texture=&T_Waterfall_Sand
  }
  n7 = VectorParameter() {group=TextureB, parameterName=TextureB_UVControl, channelNames {a=VPan, b=UPan, g=VScale, r=UScale}, defaultValue=<rgba 0.5 0.5 0 0.25>}
  n8 = Add(A=Multiply(A=ParticleRandom()) {constB=1000}, B=Time() {bOverride_Period=true, period=6666})
  n9 = Multiply(A=Add(A=MaterialFunctionCall(Texture=@n6, UVControl_RGB=@n7, UVControl_A=@n7, "[Override]TimeInput"=@n8) {materialFunction=&FXCommon_TextureSample}.G)) {constB=0.5}
  n10 = TextureObjectParameter() {
   group=NoiseA
   parameterName=NoiseA_Texture
   samplerType=SAMPLERTYPE_LinearGrayscale
   sortPriority=0
   texture=&T_SimplePerlin_Greyscale
  }
  n11 = VectorParameter() {group=NoiseA, parameterName=NoiseA_UVControl, channelNames {a=VPan, b=UPan, g=VScale, r=UScale}, defaultValue=<rgba 0.3 0.3 0 0.5>}
  n12 = MaterialFunctionCall(Texture=@n10, UVControl_RGB=@n11, UVControl_A=@n11, "[Override]TimeInput"=@n8) {materialFunction=&FXCommon_TextureSample}
  n13 = VectorParameter() {
   group=NoiseA
   parameterName=NoiseA_NoiseControl
   sortPriority=60
   channelNames {a=VStrength, b=UStrength, g=VOffset, r=UOffset}
   defaultValue=<rgba -0.5 -0.5 1 1>
  }
  n14 = MaterialFunctionCall!(NoiseValue=AppendVector(A=@n12.R, B=@n12.R), Control_RGB=@n13, Control_A=@n13) {materialFunction=/Game/Effects/Niagara/MasterMaterials/Functions/Common/FXCommon_NoiseControl}
  n15 = TextureCoordinate()
  n16 = Multiply(
   A=Add(A=Multiply(A=@n5) {constB=0.5}, B=Power(Base=@n9))
   B=SphereMask(A=Add(A=@n14.NoiseValue, B=@n15), B=Constant() {r=0.5}) {attenuationRadius=0.4, hardnessPercent=1}
  )
  Multiply(
   A=Multiply(
    A=MaterialFunctionCall!(UorV=@n2.R, EdgeControl=@n3) {materialFunction=/Game/Effects/Niagara/MasterMaterials/Functions/Common/FXCommon_EdgeMask_Simple}.Mask
    B=MaterialFunctionCall(UorV=@n2.G, EdgeControl=@n3) {materialFunction=&FXCommon_EdgeMask_Simple}.Mask
   )
   B=DepthFade(Opacity=Multiply(A=@n16, B=@n1.A))
  )
  Normalize(VectorInput=@n1.RGB)
 ]
}
