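# M_Waterfall_GeoDebug_UV is a debug material in the WaterfallGenerator tool, named for inspecting the UVs of waterfall geometry.
# It takes the fractional part of the default texture coordinates, so values repeat every UV tile.
# It uses Default Lit shading and holds a scalar parameter "sss" with no visible connection to the UV math.
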Material /Game/Tools/WaterfallGenerator/Art/M_Waterfall_GeoDebug_UV
graph {
 properties {pixelDepthOffsetMode=PDOM_Legacy, usedShadingModels=Default Lit}
 nodes [
  Frac(None=TextureCoordinate())
  ScalarParameter() {defaultValue=0.5, parameterName=sss}
 ]
}
