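# Wireframe debug material in the waterfall generator tool, apparently for inspecting generated waterfall geometry.
# It renders as an unlit, translucent wireframe.
# The graph holds a scalar parameter named Color and a clamped UV checker texture sample borrowed from the weapon wrap assets, with no connections shown for either.
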Material /Game/Tools/WaterfallGenerator/Art/M_Waterfall_GeoDebug_Wire
graph {
 properties {blendMode=BLEND_Translucent, pixelDepthOffsetMode=PDOM_Legacy, shadingModel=MSM_Unlit, usedShadingModels=Unlit, wireframe=true}
 nodes [
  ScalarParameter() {defaultValue=10, parameterName=Color}
  TextureSample() {texture=/Game/Weapons/WeaponSkins/Wraps/Textures/UVCheaker_Clamped}
 ]
}
