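# M_Waterfall_geoDebug_FaceOverlap is an unlit translucent debug material in the WaterfallGenerator tool, named for showing overlapping faces in waterfall geometry.
# Its graph holds only a constant of one, a low-valued scalar parameter named "sss" and a texture coordinate node, with no connections shown.
# The name and translucent blending suggest that stacked faces read through accumulated opacity.
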Material /Game/Tools/WaterfallGenerator/Art/M_Waterfall_geoDebug_FaceOverlap
graph {
 properties {blendMode=BLEND_Translucent, pixelDepthOffsetMode=PDOM_Legacy, shadingModel=MSM_Unlit, usedShadingModels=Unlit}
 nodes [
  Constant() {r=1}
  ScalarParameter() {defaultValue=0.2, parameterName=sss}
  TextureCoordinate()
 ]
}
