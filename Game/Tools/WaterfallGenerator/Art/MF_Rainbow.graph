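# Material function for the waterfall generator that outputs a rainbow color centered on a world-position parameter, with strength, width and offset controls.
# World-space-aligned screen coordinates around that center become radial UVs that sample the rainbow texture, clamped and edge-masked into a band.
# Scrolling noise modulates the colors, while a tiled sand texture's red channel adds grain and varies desaturation.
# The result fades out with world-space distance from the center.
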MaterialFunction /Game/Tools/WaterfallGenerator/Art/MF_Rainbow
graph {nodes [
  n1 = VectorParameter() {parameterName=RainbowCenterPos, defaultValue=<rgba 48276 -67300 2688 1>}
  n2 = Subtract(A=Distance(A=WorldPosition().XYZ, B=@n1.RGB)) {constB=500}
  n3 = Power(Base=Saturate(None=Divide(A=@n2) {constB=1000})) {constExponent=20}
  n4 = ScalarParameter() {defaultValue=0.1, parameterName=Rainbow_Str}
  n5 = Multiply(A=TextureCoordinate(), B=ScalarParameter() {defaultValue=1.5, parameterName=Rainbow_SandScale})
  n6 = Multiply(A=Add(A=TextureSample(UVs=@n5) {samplerType=SAMPLERTYPE_Normal, texture=&/T_Waterfall_Sand}.R)) {constB=0.5}
  n7 = Multiply(A=Power(Base=@n6) {constExponent=1.7}) {constB=7}
  n8 = MaterialFunctionCall(Float2=MaterialFunctionCall(Vector or UVs=MaterialFunctionCall!(Center Point World Position=@n1.RGB) {materialFunction=/Engine/Functions/Engine_MaterialFunctions02/UVs/WorldSpaceAlignedScreenCoordinates}.Offset UVs) {materialFunction=&VectorToRadialValue}.Radial Coordinates) {materialFunction=&BreakOutFloat2Components}
  n9 = OneMinus(None=@n8.G)
  n10 = ScalarParameter() {defaultValue=0.3, parameterName=Rainbow_Width}
  n11 = Multiply(A=@n10) {constB=2000}
  n12 = Add(
   A=MaterialFunctionCall(
    UVs=MaterialFunctionCall(X=@n9, Y=@n8.R) {materialFunction=&MakeFloat2}.Result
    Texture Scale=AppendVector(A=@n11, B=@n11)
   ) {materialFunction=&ScaleUVsByCenter}.UVs
   B=Divide(A=ScalarParameter() {defaultValue=1.75, parameterName=Rainbow_Offset}, B=Divide(A=@n10) {constB=0.3})
  )
  n13 = Time() {bOverride_Period=true, period=3333}
  n14 = TextureSample(UVs=Add(A=Multiply(A=@n12), B=Multiply(A=@n13) {constB=0.1})) {samplerType=SAMPLERTYPE_LinearGrayscale, texture=&/T_Waterfall_NoiseB}
  n15 = TextureCoordinate()
  n16 = DDX(Value=@n15)
  n17 = DDY(Value=@n15)
  n18 = Frac(None=@n12)
  n19 = Constant3Vector() {constant=<rgba 0 2 1 1>}
  n20 = LinearInterpolate(
   A=Constant() {r=0.8}
   B=Constant()
   Alpha=OneMinus(None=MaterialFunctionCall!(UorV=@n18, EdgeControl=@n19) {materialFunction=/Game/Effects/Niagara/MasterMaterials/Functions/Common/FXCommon_EdgeMask_Simple}.Mask)
  ) {constB=0}
  n21 = Multiply(A=@n6) {constB=-0.3}
  n22 = Multiply(
   A=Add(A=Power(Base=@n14.R) {constExponent=1.5}) {constB=0.5}
   B=Desaturation(
    None=TextureSample!(UVs=@n12, Tex, "DDX(UVs)"=@n16, "DDY(UVs)"=@n17, Apply View MipBias) {mipValueMode=TMVM_Derivative, texture=&/T_Waterfall_Rainbow}.RGB
    Fraction=Add(A=@n20, B=@n21)
   )
  )
  n23 = Constant2Vector() {r=1}
  n24 = Constant3Vector() {constant=<rgba 0 5 1 1>}
  n25 = LinearInterpolate(
   B=Multiply(A=Add(A=@n7), B=Multiply(A=@n22) {constB=0.2})
   Alpha=Multiply(
    A=MaterialFunctionCall!(TexCoord=@n12, ClampChannel=@n23) {materialFunction=/Game/Effects/Fort_Effects/MaterialFunctions/MF_FX_TilingMethod}.Result
    B=MaterialFunctionCall(UorV=@n18, EdgeControl=@n24) {materialFunction=&FXCommon_EdgeMask_Simple}.Mask
   )
  )
  FunctionOutput(None=Multiply(A=OneMinus(None=@n3), B=Multiply(A=@n4, B=@n25)))
 ]}
